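# Physics asset for the cinematic juggling prop skeletal mesh in the juggling ball cosmetics animation props.
# Each of the five ball bones carries an identical sphere body, and the root bone carries a small capsule.
# No constraints link the bodies, so the asset defines only per-bone collision shapes with no joints between the balls.
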PhysicsAsset /BRCosmetics/Animation/Props/JugglingBall/Meshes/Cine_Juggling_Physics
bodies {
 ball_01 [{shapeType=Sphere, radius=5.25}]
 ball_02 [{shapeType=Sphere, radius=5.25}]
 ball_03 [{shapeType=Sphere, radius=5.25}]
 ball_04 [{shapeType=Sphere, radius=5.25}]
 ball_05 [{shapeType=Sphere, radius=5.25}]
 root [{shapeType=Capsule, length=4.0865, radius=4.0051, center=<-0.0018 -0.009 -0.0057>, rotation=<rot 88.5496 68.3053 158.3117>}]
}
constraints []
